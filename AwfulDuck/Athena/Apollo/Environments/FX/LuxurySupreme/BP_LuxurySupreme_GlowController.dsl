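; Glow controller Blueprint for the LuxurySupreme ship environment FX, driving ship and updraft mesh materials between idle and abducting looks.
; BeginPlay creates dynamic instances for energy-parent ship materials and updraft meshes; UpdateMaterialValues sets their fade, glow and tint colors.
; Warm-up and cool-down timelines drive those values, then the server sets replicated isAbducting, whose OnRep snaps the end state.
; WarmUpTime is unused, and cool-down passes glow and fade in swapped order.
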(event EventBeginPlay
  (CallFunction|InitMaterialReferences))

(event Custom|BeginWarmUp
  (params (WarmUpTime :type Float%20%28double-precision%29))
  (Components|Timeline|SetPlayRate (Variables|BP_LuxurySupreme_GlowController|GetPulseRamp) 1.0)
  (bind (_direction _newtrack_0) (Timeline|PulseRamp PlayFromStart)
    (:Update
      (CallFunction|UpdateMaterialValues (Math|Float|Lerp (Variables|Default|GetMinFade) (Variables|Default|GetMaxFade) _newtrack_0) (Math|Float|Lerp (Variables|Default|GetMinGlow) (Variables|Default|GetMaxGlow) _newtrack_0)))
    (:Finished
      (if (Networking|IsServer)
        (Variables|Default|SetisAbducting true)))))

(event Custom|BeginCoolDown
  (Components|Activation|Activate (Variables|Default|GetNS_LuxurySupreme_FlashFadeOut) false)
  (bind (_direction _newtrack_0) (Timeline|Timeline_0 PlayFromStart)
    (:Update
      (CallFunction|UpdateMaterialValues (Math|Float|Lerp (Variables|Default|GetMinGlow) (Variables|Default|GetMaxGlow) _newtrack_0) (Math|Float|Lerp (Variables|Default|GetMinFade) (Variables|Default|GetMaxFade) _newtrack_0)))
    (:Finished
      (if (Networking|IsServer)
        (Variables|Default|SetisAbducting false)))))

(fn InitMaterialReferences
  (Utilities|FlowControl|Sequence
    (:then_0
      (for _array_element (Variables|Glow|GetShipMeshes)
        (for (_array_element_2 _array_index) (Rendering|Material|GetMaterials (Variables|/Script/Engine.StaticMeshActor|GetStaticMeshComponent _array_element))
          (bind _returnvalue (Rendering|Material|GetBaseMaterial _array_element_2))
          (if (Utilities|Equal(Object) _returnvalue "/Game/Environments/Apollo/Sets/Skyfire/Materials/Parents/M_Skyfire_Energy_Parent.M_Skyfire_Energy_Parent")
            (Utilities|Casting|CastToMaterialInstanceDynamic _array_element_2
              (:then
                (Utilities|Array|AddUnique (Variables|Default|GetGlowMaterials) _array_element_2))
              (:CastFailed
                (bind _returnvalue_2 (Rendering|Material|CreateDynamicMaterialInstance _array_element_2 "None" "None" :!class /Script/Engine.KismetMaterialLibrary))
                (Rendering|Material|SetMaterial (Variables|/Script/Engine.StaticMeshActor|GetStaticMeshComponent _array_element) _array_index _returnvalue_2)
                (Utilities|Array|AddUnique (Variables|Default|GetGlowMaterials) _returnvalue_2)))))))
    (:then_1
      (for _array_element_3 (Variables|Glow|GetUpdraftMeshes)
        (Utilities|IsValid _array_element_3
          (:Is%20Valid
            (Rendering|Material|CreateDynamicMaterialInstance (Variables|/Script/Engine.StaticMeshActor|GetStaticMeshComponent _array_element_3) 0 (Rendering|Material|GetMaterial (Variables|/Script/Engine.StaticMeshActor|GetStaticMeshComponent _array_element_3) 0) "None" :!class /Script/Engine.PrimitiveComponent)))))))

(fn OnRep_isAbducting
  (CallFunction|UpdateMaterialValues (select (Variables|Default|GetisAbducting) (Variables|Default|GetMaxFade) (Variables|Default|GetMinFade)) (select (Variables|Default|GetisAbducting) (Variables|Default|GetMaxGlow) (Variables|Default|GetMinGlow))))

(fn UpdateMaterialValues
  (params (Fade :type Float%20%28double-precision%29) (Glow :type Float%20%28double-precision%29))
  (Utilities|FlowControl|Sequence
    (:then_0
      (for _array_element (Variables|Default|GetGlowMaterials)
        (bind _asmaterial_instance_dynamic (Utilities|Casting|CastToMaterialInstanceDynamic _array_element))
        (Rendering|Material|SetScalarParameterValue _asmaterial_instance_dynamic "AbductionFade" Fade :!class /Script/Engine.MaterialInstanceDynamic)
        (Rendering|Material|SetScalarParameterValue _asmaterial_instance_dynamic "AbductionGlowMult" Glow :!class /Script/Engine.MaterialInstanceDynamic)))
    (:then_1
      (for _array_element_2 (Variables|Glow|GetUpdraftMeshes)
        (bind _asmaterial_instance_dynamic_2 (Utilities|Casting|CastToMaterialInstanceDynamic (Rendering|Material|GetMaterial (Variables|/Script/Engine.StaticMeshActor|GetStaticMeshComponent _array_element_2) 0)))
        (Rendering|Material|SetVectorParameterValue _asmaterial_instance_dynamic_2 "CubeLightingColor" (Math|Color|Lerp(LinearColor) (Variables|Default|GetUpdraftTintInactiveColor) (Variables|Default|GetUpdraftTintActiveColor) Fade) :!class /Script/Engine.MaterialInstanceDynamic)
        (bind _returnvalue (Rendering|Material|GetBaseMaterial (Rendering|Material|GetMaterial (Variables|/Script/Engine.StaticMeshActor|GetStaticMeshComponent _array_element_2) 0)))
        (if (Utilities|Equal(Object) _returnvalue "/Game/Effects/Niagara/MasterMaterials/Wind/M_Mesh_Wind.M_Mesh_Wind")
          (Rendering|Material|SetVectorParameterValue _asmaterial_instance_dynamic_2 "Color" (Math|Color|Lerp(LinearColor) (Variables|Default|GetWindInactiveColor) (Variables|Default|GetWindActiveColor) Fade) :!class /Script/Engine.MaterialInstanceDynamic)
          (Rendering|Material|SetVectorParameterValue _asmaterial_instance_dynamic_2 "ColorBase" (Math|Color|Lerp(LinearColor) (Variables|Default|GetWindInactiveColorBase) (Variables|Default|GetWindActiveColorBase) Fade) :!class /Script/Engine.MaterialInstanceDynamic))))))

(fn ConstructionScript)
