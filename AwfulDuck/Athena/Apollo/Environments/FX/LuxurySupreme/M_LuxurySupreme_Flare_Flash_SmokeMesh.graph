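# Additive unlit Niagara mesh-particle material for the smoke mesh of the LuxurySupreme flare flash effect.
# Radial UVs drive panned fire noise and inky smoke textures, which multiply into an animated mask shaped by axis-aligned and inverted Fresnel.
# Camera distance fade, particle alpha and a sphere mask modulate that mask, while particle color feeds an emissive adjustment function.
# A separate branch scales panned distortion normals and a smooth texture by the WaveIntensity_Top dynamic parameter.
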Material /AwfulDuck/Athena/Apollo/Environments/FX/LuxurySupreme/M_LuxurySupreme_Flare_Flash_SmokeMesh
graph {
 properties {
  bUsedWithNiagaraMeshParticles=true
  blendMode=BLEND_Additive
  pixelDepthOffsetMode=PDOM_Legacy
  previewMesh=/Game/Athena/Apollo/Environments/FX/LuxurySupreme/SM_LuxurySupreme_FlashFade
  shadingModel=MSM_Unlit
  usedShadingModels=Unlit
 }
 nodes [
  n1 = TextureCoordinate()
  n2 = Constant2Vector() {g=-0.5, r=-0.5}
  n3 = MaterialFunctionCall(Vector or UVs=Multiply(A=Add(A=@n1, B=@n2), B=ScalarParameter() {defaultValue=2, parameterName=RingTiler})) {materialFunction=&VectorToRadialValue}
  n4 = Multiply(A=Time(), B=ScalarParameter() {defaultValue=2, parameterName=Speed})
  n5 = ComponentMask(None=TextureSample(UVs=Panner(Coordinate=@n3.Radial Coordinates, Time=@n4) {speedY=0.25}) {samplerType=SAMPLERTYPE_Normal, texture=&T_Weap_Distort_N_02}.RGB) {g=true, r=true}
  n6 = TextureSample(UVs=@n3.Radial Coordinates) {texture=&T_Smooth_tile}
  Add(B=Multiply(
    A=AppendVector(A=@n5, B=@n6.R)
    B=DynamicParameter() {defaultValue=<rgba 1 50 250 1>, paramNames [Glow, WaveIntensity_Bot, WaveIntensity_Top, Param4]}.WaveIntensity_Top
   ) {constB=10})
  n7 = ParticleColor()
  MaterialFunctionCall(Emissive=@n7.RGB) {materialFunction=&MF_FX_EmissiveAdjustments}
  n8 = TextureCoordinate()
  n9 = Constant2Vector() {g=-0.5, r=-0.5}
  n10 = MaterialFunctionCall(Vector or UVs=Multiply(A=Add(A=@n8, B=@n9), B=ScalarParameter() {defaultValue=2, parameterName=RingTiler})) {materialFunction=&VectorToRadialValue}
  n11 = TextureSample(UVs=Panner(Coordinate=@n10.Radial Coordinates, Time=@n4) {speedY=0.15}) {texture=/Game/Global/Textures/Default/Noise/T_FireNoise_02}
  n12 = TextureSample(UVs=Panner(Coordinate=@n10.Radial Coordinates, Time=@n4) {speedY=0.25}) {texture=&T_FireNoise_02}
  n13 = Clamp(None=Power(Base=Multiply(A=@n11.G, B=@n12.G), Exp=ScalarParameter() {defaultValue=1, parameterName=Exp}) {constExponent=4})
  n14 = Power(Base=TextureSample(UVs=Panner(Coordinate=@n10.Radial Coordinates, Time=@n4) {speedY=0.25}) {texture=&T_Inky_Smoke_Tile}.G) {constExponent=0.8}
  n15 = ScalarParameter() {defaultValue=1, parameterName=MidDetail_Intensity}
  n16 = MaterialFunctionCall!(
   Axis=VectorParameter() {parameterName="Axis (V3)", defaultValue=<rgba 0 1 0 0>}.RGB
   Exponent=ScalarParameter() {defaultValue=1, parameterName=ExponentIn}
   Normal
  ) {materialFunction=/Game/Effects/Fort_Effects/Materials/_Functions/MF_AxisAlignedFrensnel}
  n17 = OneMinus(None=Fresnel(
    ExponentIn=ScalarParameter() {defaultValue=0.25, parameterName=ExponentIn_1}
    BaseReflectFractionIn=ScalarParameter() {defaultValue=0.04, parameterName=BaseReflectFractionIn}
   ))
  n18 = LinearInterpolate(B=Multiply(A=Multiply(A=@n14, B=@n15), B=Multiply(A=@n16.Fresnel, B=@n17)), Alpha=ScalarParameter() {defaultValue=0.99, parameterName=Alpha}) {constA=1}
  n19 = ScalarParameter() {defaultValue=250, parameterName="Fade_Start_Distance (S)"}
  n20 = ScalarParameter() {defaultValue=125, parameterName="Fade_End_Distance (S)"}
  n21 = Multiply(
   A=Multiply(A=Multiply(A=@n13, B=@n18), B=ScalarParameter() {defaultValue=2, parameterName=B}) {constB=15}
   B=LinearInterpolate(Alpha=MaterialFunctionCall(Fade_Start_Distance=@n19, Fade_End_Distance=@n20) {materialFunction=&MF_Distance_SyncedFade_FX}.Result) {constA=0.25, constB=0.35}
  )
  Multiply(
   A=Multiply(A=@n7.A, B=Clamp(None=@n21))
   B=SphereMask(
    A=TextureCoordinate()
    B=Constant2Vector() {g=0.5, r=0.5}
    Radius=ScalarParameter() {defaultValue=0.45, parameterName=Radius}
    Hardness=ScalarParameter() {parameterName=Hardness}
   )
  )
 ]
}
