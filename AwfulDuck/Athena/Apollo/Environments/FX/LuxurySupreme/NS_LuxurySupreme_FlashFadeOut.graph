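# Niagara environment effect for LuxurySupreme: a short, large-scale flash that shifts from green toward pink-magenta as it fades out.
# Burst-spawned meshes form two vertically offset light cones, a flash mesh with a light renderer, and two shockwave layers.
# A brief GPU emitter adds large, curl-noise-driven sprites on a cylinder surface, with separate fog color and size attributes.
# Age-driven curves control color, alpha and growth, and camera offsets pull layers forward.
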NiagaraSystem /AwfulDuck/Athena/Apollo/Environments/FX/LuxurySupreme/NS_LuxurySupreme_FlashFadeOut
dependencies {
 usedDataInterfaces [&NiagaraDataInterfaceCurve, &NiagaraDataInterfaceVectorCurve, &NiagaraDataInterfaceColorCurve]
 usedDynamicInputs []
 usedModules [
  &self:SystemSpawnScript.SystemScriptSource.SystemScriptGraph.NiagaraNodeAssignment_0.SetVariables_#1
  &SystemState
 ]
}
emitters {
 ConeLight001 {
  emitterSpawnScript [/SetVariables(Emitter.LightConeOverallSize=2.2, Emitter.RingUV_Distortion_Input=1, Emitter.UseParticleColorInput=1) {isSetParameters=true}]
  emitterUpdateScript [
   &EmitterState(Life Cycle Mode=::NewEnumerator1, Loop Behavior=::NewEnumerator1, Loop Duration=1, UseLoopDelay=false)
   &SpawnBurst_Instantaneous(Spawn Count=1, Use Spawn Probability=false)
   &SpawnBurst_Instantaneous(Spawn Count=1, Spawn Group=1, Use Spawn Probability=false) {module=SpawnBurst_Instantaneous001}
  ]
  particleSpawnScript [
   /SetVariables(Particles.TopConeSize:Vector3f=&Multiply_VectorByFloat(Vector=<2 14 14>, Float=$Emitter.LightConeOverallSize)) {isSetParameters=true}
   /SetVariables(Particles.DyParam_Top:Vector4f=<2 0.5 4 0.5>) {isSetParameters=true}
   &InitializeParticle(Color=<rgba 2.5001 10 0.2416 1>, Lifetime=2, UsePositionOffset=false, Write Material Random=true)
   &InitialMeshOrientation(
    Orientation Axis=<0 0 1>
    Orientation Vector=&Lerp_Vector(A=<-1 0 0>, B=<-1 0 0>, Alpha=&MaskFloatBySpawnGroup(Masked Float=0, Passthrough Float=1))
    Use Orientation Vector=true
    Use Rotation Vector=false
   )
   &SystemLocation(MaskSpawnGroup=true, Offset=<0 0 3000>)
   &SystemLocation(MaskSpawnGroup=true, Offset=<0 0 -3000>, Spawn Group Mask=1) {module=SystemLocation001}
   &CameraOffset(Camera Offset Amount=2500)
  ]
  particleUpdateScript [
   &ParticleState()
   /SetVariables(Particles.GradientColor:Vector3f=&Multiply_Vector(
     A=&MakeVectorFromLinearColorRGB(Color=&MakeLinearColorFromVectorAndFloat(
       "Vector (RGB)"=&MakeVectorFromLinearColorRGB(Color=&ColorFromCurve(ColorCurve=DataInterface() {propertyValues {
           AlphaCurve {keys [<at 0 1 RCIM_Cubic>,<at 0.9977 1 RCIM_Cubic>]}
           BlueCurve {keys [<at 0.2046 0.1191 RCIM_Cubic>,<at 0.3221 0.5359>]}
           GreenCurve {keys [<at 0.2046 1 RCIM_Cubic>,<at 0.3221 0>]}
           RedCurve {keys [<at 0.2046 0 RCIM_Cubic>,<at 0.3221 1>]}
          }}))
       "Float (Alpha)"=1
      ))
     B=&VectorFromFloat(Value=&FloatFromCurve(FloatCurve=DataInterface() {propertyValues {Curve {keys [<at 0 1>,<at 1 0>]}}}))
    )) {isSetParameters=true}
   /SetVariables(Particles.Color=&MakeLinearColorFromVectorAndFloat(
     "Vector (RGB)"=&Lerp_Vector(
      A=&Multiply_VectorByFloat(Vector=$Particles.GradientColor, Float=2)
      B=&Multiply_VectorByFloat(Vector=$Particles.GradientColor, Float=15)
      Alpha=1
     )
     "Float (Alpha)"=&FloatFromCurve(
      FloatCurve=DataInterface() {propertyValues {Curve {keys [
          {interpMode=RCIM_Cubic, tangentMode=RCTM_User, time=-0.0063, value=0.0184}
          {arriveTangent=-2.5281, interpMode=RCIM_Cubic, leaveTangent=-2.5281, tangentMode=RCTM_User, time=0.1, value=1}
          {interpMode=RCIM_Cubic, tangentMode=RCTM_User, time=1.0011, value=0.002}
         ]}}}
      Scale Curve=0.25
     )
    )) {isSetParameters=true}
   /SetVariables(Particles.SizeUpdate:Vector3f=&Multiply_Vector(
     A=$Particles.TopConeSize
     B=&VectorFromFloat(Value=&FloatFromCurve(FloatCurve=DataInterface() {propertyValues {Curve {keys [<at 0 1>,<at 1 2>]}}}))
    )) {isSetParameters=true}
  ]
  rendererClasses [&NiagaraMeshRendererProperties]
  renderers [{rendererClass=&NiagaraMeshRendererProperties, rendererIndex=0}]
 }
 FN_Mesh {
  emitterUpdateScript [
   &EmitterState()
   &SpawnBurst_Instantaneous(Spawn Count=1, Use Spawn Probability=false)
   &SpawnBurst_Instantaneous(Spawn Count=1, Spawn Group=1, Use Spawn Probability=false) {module=SpawnBurst_Instantaneous001}
  ]
  particleSpawnScript [
   /Niagara/Modules/Spawn/Initialization/InitializeParticle(
    Color=<rgba 0 150 10.359 1>
    Lifetime=&UniformRangedFloat(Minimum=0.5, Maximum=1)
    Module.Write Color=true
    Module.Write Lifetime=true
    Module.Write Mass=true
    Module.Write Material Random=true
    Module.Write Position=false
    Module.Write Scale=true
    Module.Write SpriteRotation=false
    Module.Write SpriteSize=false
    Position=&SimulationPosition()
   ) {enabled=false}
   /Niagara/Modules/Spawn/Initialization/V2/InitializeParticle(Color=<rgba 25 25 25 2>, Mesh Scale=<1 1 1.3>, Mesh Scale Mode=::NewEnumerator4, UsePositionOffset=false, Write Material Random=true) {module=InitializeParticle001}
   &SystemLocation(MaskSpawnGroup=false, Offset=$System.SpawnOffset)
   &CameraOffset(Camera Offset Amount=100)
   /SetVariables(Particles.LightPos:Vector3f=&Lerp_Vector(A=<0 0 -15000>, B=<1 1 5000>, Alpha=&MaskFloatBySpawnGroup(Masked Float=0, Passthrough Float=1))) {isSetParameters=true}
  ]
  particleUpdateScript [
   &ParticleState()
   &ScaleMeshSize(Scale Factor=&VectorFromFloat(Value=&FloatFromCurve(FloatCurve=DataInterface() {propertyValues {Curve {keys [{tangentMode=RCTM_User, time=0, value=1.1}, {tangentMode=RCTM_User, time=1, value=1.2}]}}})))
   &ScaleColor(
    Scale Alpha=&FloatFromCurve(FloatCurve=DataInterface() {propertyValues {Curve {keys [
         {arriveTangent=9.0725, interpMode=RCIM_Cubic, leaveTangent=9.0725, tangentMode=RCTM_User, time=0, value=0}
         {arriveTangent=-2.5249, interpMode=RCIM_Cubic, leaveTangent=-2.5249, tangentMode=RCTM_User, time=0.1, value=1}
         {interpMode=RCIM_Cubic, tangentMode=RCTM_User, time=0.5, value=0.35}
         {arriveTangent=-0.0405, interpMode=RCIM_Cubic, leaveTangent=-0.0405, tangentMode=RCTM_User, time=1, value=0}
        ]}}})
    Scale RGB=&VectorFromCurve(VectorCurve=DataInterface() {propertyValues {
       XCurve {keys [
         {interpMode=RCIM_Cubic, tangentMode=RCTM_User, time=0.25, value=0}
         {interpMode=RCIM_Cubic, tangentMode=RCTM_User, time=0.8, value=1}
         <at 1 0>
        ]}
       YCurve {keys [{interpMode=RCIM_Cubic, tangentMode=RCTM_User, time=0.3, value=1}, {interpMode=RCIM_Cubic, tangentMode=RCTM_User, time=1, value=0}]}
       ZCurve {keys [
         {interpMode=RCIM_Cubic, tangentMode=RCTM_User, time=0.25, value=0}
         {interpMode=RCIM_Cubic, tangentMode=RCTM_User, time=0.8, value=1}
         <at 1 0>
        ]}
      }})
    ScaleA=true
    ScaleRGB=true
   )
   /SetVariables(Particles.NewLinearColor:LinearColor=/Niagara/DynamicInputs/LinearColor/MakeLinearColorFromFloat4(Vector:Vector4f=&MakeFloat4FromVectorAndFloat(
      "Vector (RGB)"=&VectorFromCurve(
       VectorCurve=DataInterface() {propertyValues {
         XCurve {keys [
           {interpMode=RCIM_Cubic, tangentMode=RCTM_User, time=0, value=0}
           {interpMode=RCIM_Cubic, tangentMode=RCTM_User, time=0.7, value=0.75}
           {interpMode=RCIM_Cubic, tangentMode=RCTM_User, time=1, value=0}
          ]}
         YCurve {keys [
           {interpMode=RCIM_Cubic, tangentMode=RCTM_User, time=0, value=1}
           {arriveTangent=-1.2904, interpMode=RCIM_Cubic, leaveTangent=-1.2904, tangentMode=RCTM_User, time=0.5, value=0.375}
           {interpMode=RCIM_Cubic, tangentMode=RCTM_User, time=1, value=0}
          ]}
         ZCurve {keys [
           {interpMode=RCIM_Cubic, tangentMode=RCTM_User, time=0, value=0}
           {interpMode=RCIM_Cubic, tangentMode=RCTM_User, time=0.7, value=0.65}
           {interpMode=RCIM_Cubic, tangentMode=RCTM_User, time=1, value=0}
          ]}
        }}
       Scale Curve=<50 50 50>
      )
      "Float (Alpha)"=&FloatFromCurve(FloatCurve=DataInterface() {propertyValues {Curve {keys [<at 0 1>,<at 1 0>]}}})
     ))) {isSetParameters=true}
  ]
  rendererClasses [&NiagaraMeshRendererProperties, &NiagaraLightRendererProperties]
  renderers [
   {rendererClass=&NiagaraMeshRendererProperties, rendererIndex=0}
   {rendererClass=&NiagaraLightRendererProperties, rendererIndex=1}
  ]
 }
 HT_LoopShockwave_CamF {
  emitterUpdateScript [
   &EmitterState()
   &SpawnBurst_Instantaneous(Spawn Count=1, Use Spawn Probability=false)
  ]
  particleSpawnScript [
   &InitializeParticle(
    Color=<rgba 3 3 3 0.3>
    Lifetime=2
    Mesh Scale=<0.175 0.175 0.175>
    Mesh Scale Mode=::NewEnumerator4
    UsePositionOffset=false
    Write Material Random=true
   ) {module=InitializeParticle001}
   &InitialMeshOrientation(Orientation Axis=<0 0 0>, Orientation Vector=<0 0 0>, Use Orientation Vector=true, Use Rotation Vector=false)
   &SystemLocation(MaskSpawnGroup=false, Offset=<0 0 -2000>) {module=SystemLocation001}
   &CameraOffset(Camera Offset Amount=1000)
  ]
  particleUpdateScript [
   &ParticleState()
   &ScaleColor(
    Scale Alpha=&FloatFromCurve(FloatCurve=DataInterface() {propertyValues {Curve {keys [
         {interpMode=RCIM_Cubic, tangentMode=RCTM_User, time=0.0023, value=0}
         {arriveTangent=2.0824, interpMode=RCIM_Cubic, leaveTangent=0.0112, tangentMode=RCTM_Break, time=0.0552, value=0.9565}
         {arriveTangent=-0.1758, interpMode=RCIM_Cubic, leaveTangent=-0.1758, tangentMode=RCTM_User, time=1.0011, value=0.002}
        ]}}})
    Scale RGB=<150 150 150>
    ScaleA=true
    ScaleRGB=true
   ) {enabled=false}
   &ScaleColor(
    Scale Alpha=&Multiply_Float(
     A=&FloatFromCurve(FloatCurve=DataInterface() {propertyValues {Curve {keys [
          {interpMode=RCIM_Cubic, tangentMode=RCTM_User, time=0, value=1}
          {interpMode=RCIM_Cubic, tangentMode=RCTM_User, time=0.5, value=1}
          {interpMode=RCIM_Cubic, tangentMode=RCTM_User, time=1, value=1}
         ]}}})
     B=&FloatFromCurve(
      FloatCurve=DataInterface() {propertyValues {Curve {keys [
          {arriveTangent=9.0725, interpMode=RCIM_Cubic, leaveTangent=9.0725, tangentMode=RCTM_User, time=0, value=0}
          {arriveTangent=-2.5249, interpMode=RCIM_Cubic, leaveTangent=-2.5249, tangentMode=RCTM_User, time=0.1408, value=0.9847}
          {arriveTangent=-0.0405, interpMode=RCIM_Cubic, leaveTangent=-0.0405, tangentMode=RCTM_User, time=1, value=0}
         ]}}}
      Scale Curve=0.5
     )
    )
    Scale RGB=&Multiply_Vector(
     A=&MakeVectorFromLinearColorRGB(Color=&ColorFromCurve(ColorCurve=DataInterface() {propertyValues {
         AlphaCurve {keys [<at 0 1 RCIM_Cubic>,<at 0.9977 1 RCIM_Cubic>]}
         BlueCurve {keys [<at 0.1175 0.0676 RCIM_Cubic>,<at 0.346 0.3089 RCIM_Cubic>]}
         GreenCurve {keys [<at 0.1175 0.3372 RCIM_Cubic>,<at 0.346 0 RCIM_Cubic>]}
         RedCurve {keys [<at 0.1175 0 RCIM_Cubic>,<at 0.346 0.3372 RCIM_Cubic>]}
        }}))
     B=&VectorFromFloat(Value=25)
    )
    ScaleA=true
    ScaleRGB=true
   ) {module=ScaleColor001}
   &DynamicMaterialParameters(
    Index 0 Param 2=-1
    Index 0 Param 3=5
    Index 0 Param 4=1
    Module.Param0WriteEnabled=true
    Module.Param1WriteEnabled=true
    Module.Param2WriteEnabled=true
    Module.Param3WriteEnabled=true
   )
  ]
  rendererClasses [&NiagaraMeshRendererProperties]
  renderers [{rendererClass=&NiagaraMeshRendererProperties, rendererIndex=0}]
 }
 HT_LoopShockwave_CamF001 {
  emitterUpdateScript [
   &EmitterState()
   &SpawnRate(SpawnRate=0, Use Spawn Probability=false)
   &SpawnBurst_Instantaneous(Spawn Count=1, Use Spawn Probability=false)
  ]
  particleSpawnScript [
   &InitializeParticle(
    Color=<rgba 3 3 3 0.3>
    Lifetime=1.25
    Mesh Scale Mode=::NewEnumerator1
    Mesh Uniform Scale Max=250
    Mesh Uniform Scale Min=250
    UsePositionOffset=false
    Write Material Random=true
   ) {module=InitializeParticle001}
   &InitialMeshOrientation(Orientation Axis=<0 0 0>, Orientation Vector=<0 0 1>, Use Orientation Vector=true, Use Rotation Vector=false) {enabled=false}
   &CameraOffset(Camera Offset Amount=1000)
  ]
  particleUpdateScript [
   &ParticleState()
   &ScaleMeshSize(Scale Factor=/Niagara/DynamicInputs/ScaleByCurve/ScaleVectorByCurve(
     InitialValue:Vector3f=<1 1 1>
     ScaleCurve:NiagaraDataInterfaceVectorCurve=DataInterface() {propertyValues {
       XCurve {keys [
         {arriveTangent=7.5627, interpMode=RCIM_Cubic, leaveTangent=3.2485, tangentMode=RCTM_User, time=0, value=0}
         {interpMode=RCIM_Cubic, tangentMode=RCTM_User, time=1, value=1}
        ]}
       YCurve {keys [
         {arriveTangent=9.2714, interpMode=RCIM_Cubic, leaveTangent=3.2485, tangentMode=RCTM_User, time=0, value=0}
         {interpMode=RCIM_Cubic, tangentMode=RCTM_User, time=1, value=1}
        ]}
       ZCurve {keys [
         {arriveTangent=8.0047, interpMode=RCIM_Cubic, leaveTangent=3.2485, tangentMode=RCTM_User, time=0, value=0}
         {interpMode=RCIM_Cubic, tangentMode=RCTM_User, time=1, value=1}
        ]}
      }}
     CurveIndex:NiagaraFloat=$Particles.NormalizedAge
    )) {enabled=false}
   &ScaleColor(
    Scale Alpha=&FloatFromCurve(FloatCurve=DataInterface() {propertyValues {Curve {keys [
         {interpMode=RCIM_Cubic, tangentMode=RCTM_User, time=0.0023, value=0}
         {arriveTangent=2.0824, interpMode=RCIM_Cubic, leaveTangent=0.0112, tangentMode=RCTM_Break, time=0.0552, value=0.9565}
         {arriveTangent=-0.1758, interpMode=RCIM_Cubic, leaveTangent=-0.1758, tangentMode=RCTM_User, time=1.0011, value=0.002}
        ]}}})
    Scale RGB=<150 150 150>
    ScaleA=true
    ScaleRGB=true
   ) {enabled=false}
   &ScaleColor(
    Scale Alpha=&Multiply_Float(
     A=&FloatFromCurve(FloatCurve=DataInterface() {propertyValues {Curve {keys [
          {interpMode=RCIM_Cubic, tangentMode=RCTM_User, time=0, value=1}
          {interpMode=RCIM_Cubic, tangentMode=RCTM_User, time=0.5, value=1}
          {interpMode=RCIM_Cubic, tangentMode=RCTM_User, time=1, value=1}
         ]}}})
     B=&FloatFromCurve(
      FloatCurve=DataInterface() {propertyValues {Curve {keys [
          {arriveTangent=9.0725, interpMode=RCIM_Cubic, leaveTangent=9.0725, tangentMode=RCTM_User, time=0, value=0}
          {arriveTangent=-2.5249, interpMode=RCIM_Cubic, leaveTangent=-2.5249, tangentMode=RCTM_User, time=0.1408, value=0.9847}
          {arriveTangent=-0.0405, interpMode=RCIM_Cubic, leaveTangent=-0.0405, tangentMode=RCTM_User, time=1, value=0}
         ]}}}
      Scale Curve=0.5
     )
    )
    Scale RGB=&Multiply_Vector(
     A=&MakeVectorFromLinearColorRGB(Color=&ColorFromCurve(ColorCurve=DataInterface() {propertyValues {
         AlphaCurve {keys [<at 0 1 RCIM_Cubic>,<at 0.9977 1 RCIM_Cubic>]}
         BlueCurve {keys [<at 0.1175 0.0676 RCIM_Cubic>,<at 0.4731 0.2529 RCIM_Cubic>]}
         GreenCurve {keys [<at 0.1175 0.3372 RCIM_Cubic>,<at 0.4731 0 RCIM_Cubic>]}
         RedCurve {keys [<at 0.1175 0 RCIM_Cubic>,<at 0.4731 0.3372 RCIM_Cubic>]}
        }}))
     B=&VectorFromFloat(Value=10)
    )
    ScaleA=true
    ScaleRGB=true
   ) {module=ScaleColor001}
   &DynamicMaterialParameters(
    Index 0 Param 2=-1
    Index 0 Param 3=5
    Index 0 Param 4=1
    Module.Param0WriteEnabled=true
    Module.Param1WriteEnabled=true
    Module.Param2WriteEnabled=true
    Module.Param3WriteEnabled=true
   )
  ]
  rendererClasses [&NiagaraMeshRendererProperties]
  renderers [{rendererClass=&NiagaraMeshRendererProperties, rendererIndex=0}]
 }
 Pop_Sparks {
  simTarget=GPUComputeSim
  emitterUpdateScript [
   &EmitterState(Life Cycle Mode=::NewEnumerator1, Loop Behavior=::NewEnumerator1, Loop Duration=0.1, UseLoopDelay=false)
   &SpawnRate(SpawnRate=15, Use Spawn Probability=true)
  ]
  particleSpawnScript [
   /Niagara/Modules/Spawn/Initialization/InitializeParticle(
    Color=&LerpLinearColors(StartColor=<rgba 0 0 0 1>, EndColor=<rgba 0 0 0 1>, LerpFactor=1)
    Lifetime=&Multiply_Float(A=&UniformRangedFloat(Minimum=1, Maximum=2), B=1)
    Module.Write Color=true
    Module.Write Lifetime=true
    Module.Write Mass=true
    Module.Write Material Random=true
    Module.Write Position=false
    Module.Write Scale=false
    Module.Write SpriteRotation=false
    Module.Write SpriteSize=true
    Position=&SimulationPosition()
    Sprite Size=<5000 5000>
   ) {enabled=false}
   /Niagara/Modules/Spawn/Initialization/V2/InitializeParticle(
    Lifetime Max=1.5
    Lifetime Min=0.7
    Lifetime Mode=::NewEnumerator1
    Sprite Rotation Mode=::NewEnumerator2
    Sprite Size Mode=::NewEnumerator1
    Uniform Sprite Size Max=8000
    Uniform Sprite Size Min=5000
    UsePositionOffset=false
    Write Material Random=true
   ) {module=InitializeParticle001}
   &AddVelocityFromPoint(
    Module.Use Falloff Distance=false
    Origin Offset=&UniformRangedVector(Minimum=<-1 -1 -100>, Maximum=<1 1 -100>)
    Origin Offset Coordinate Space=::Simulation
    Velocity Strength=&UniformRangedFloat(Minimum=0, Maximum=0)
   )
   /Niagara/Modules/Update/Forces/CurlNoiseForce(Noise Frequency=0.25, Noise Strength=25, Pan Noise=true, Pan Noise Field=<0 0 -5>) {module=CurlNoiseForce001}
   &ApplyInitialForces()
   &CylinderLocation(
    Cylinder Height=500
    Cylinder Radius=3000
    MaskSpawnGroup=false
    Module.Enable Lathe Profile=false
    Module.Override Randomness=false
    Module.Surface Only=true
    Surface Only Band Thickness=1
   )
   /SetVariables(Particles.NewNiagaraFloat:NiagaraFloat=&Multiply_Float(
     A=&MakeFloatFromLinearColor(LinearColor=$Particles.SampleStaticMesh.MeshColor, Channel:ENiagaraLinearColor_Channels=::NewEnumerator0)
     B=1500
    )) {enabled=false, isSetParameters=true}
   &CameraOffset(Camera Offset Amount=100)
   &AddVelocityFromPoint(Module.Use Falloff Distance=false, Velocity Strength=250) {module=AddVelocityFromPoint001}
  ]
  particleUpdateScript [
   &ParticleState()
   /SetVariables(Particles.GradientColor:Vector3f=&Multiply_Vector(
     A=&MakeVectorFromLinearColorRGB(Color=&MakeLinearColorFromVectorAndFloat(
       "Vector (RGB)"=&MakeVectorFromLinearColorRGB(Color=&ColorFromCurve(ColorCurve=DataInterface() {propertyValues {
           AlphaCurve {keys [<at 0 1 RCIM_Cubic>,<at 0.9977 1 RCIM_Cubic>]}
           BlueCurve {keys [<at 0.1175 0.0402 RCIM_Cubic>,<at 0.6953 0.8264>,<at 0.9071 0.521 RCIM_Cubic>]}
           GreenCurve {keys [<at 0.1175 0.3372 RCIM_Cubic>,<at 0.6953 0>,<at 0.9071 0.4852 RCIM_Cubic>]}
           RedCurve {keys [<at 0.1175 0 RCIM_Cubic>,<at 0.6953 1>,<at 0.9071 0.807 RCIM_Cubic>]}
          }}))
       "Float (Alpha)"=1
      ))
     B=&VectorFromFloat(Value=&FloatFromCurve(FloatCurve=DataInterface() {propertyValues {Curve {keys [<at 0 1>,<at 1 0>]}}}, Scale Curve=5))
    )) {isSetParameters=true}
   /SetVariables(Particles.Color=&MakeLinearColorFromVectorAndFloat(
     "Vector (RGB)"=&Lerp_Vector(
      A=&Multiply_VectorByFloat(Vector=$Particles.GradientColor, Float=0.25)
      B=&Multiply_VectorByFloat(Vector=$Particles.GradientColor, Float=0.25)
      Alpha=1
     )
     "Float (Alpha)"=&FloatFromCurve(FloatCurve=DataInterface() {propertyValues {Curve {keys [
          {interpMode=RCIM_Cubic, tangentMode=RCTM_User, time=-0.0063, value=0.0184}
          {arriveTangent=-0.8483, interpMode=RCIM_Cubic, leaveTangent=-0.8483, tangentMode=RCTM_User, time=0.1, value=0.95}
          {arriveTangent=-5.0978, interpMode=RCIM_Cubic, leaveTangent=-5.0978, tangentMode=RCTM_User, time=0.2818, value=1.2143}
          {arriveTangent=-0.2783, interpMode=RCIM_Cubic, leaveTangent=-0.2783, tangentMode=RCTM_User, time=0.3, value=0.55}
          {arriveTangent=0.3112, interpMode=RCIM_Cubic, leaveTangent=0.3112, tangentMode=RCTM_User, time=0.5407, value=1.3419}
          {interpMode=RCIM_Cubic, tangentMode=RCTM_User, time=0.6, value=0.55}
          {interpMode=RCIM_Cubic, tangentMode=RCTM_User, time=0.7, value=0.7}
          {arriveTangent=-3.7119, interpMode=RCIM_Cubic, leaveTangent=-3.7119, tangentMode=RCTM_User, time=0.7242, value=0.3749}
          {interpMode=RCIM_Cubic, tangentMode=RCTM_User, time=1.0011, value=0.002}
         ]}}})
    )) {isSetParameters=true}
   /SetVariables(Particles.FogColor:LinearColor=&MakeLinearColorFromVectorAndFloat(
     "Vector (RGB)"=&Lerp_Vector(
      A=&Multiply_VectorByFloat(Vector=$Particles.GradientColor, Float=1)
      B=&Multiply_VectorByFloat(Vector=$Particles.GradientColor, Float=5)
      Alpha=1
     )
     "Float (Alpha)"=&FloatFromCurve(FloatCurve=DataInterface() {propertyValues {Curve {keys [
          {interpMode=RCIM_Cubic, tangentMode=RCTM_User, time=-0.0063, value=0.0184}
          {arriveTangent=-5.0978, interpMode=RCIM_Cubic, leaveTangent=-5.0978, tangentMode=RCTM_User, time=0.2818, value=1.2143}
          {interpMode=RCIM_Cubic, tangentMode=RCTM_User, time=1.0011, value=0.002}
         ]}}})
    )) {isSetParameters=true}
   &CurlNoiseForce(
    Noise Frequency=1
    Noise Quality / Cost=::NewEnumerator0
    Noise Strength=&FloatFromCurve(FloatCurve=DataInterface() {propertyValues {Curve {keys [<at 0 1>,<at 0.25 0>]}}})
    Pan Noise=true
    Pan Noise Field=<0 10 0>
   )
   &SolveForcesAndVelocity(Clamp Velocity=false, Limit Acceleration=false)
   &ScaleSpriteSize(Scale Factor=&Vector2DFromFloat(Value=&FloatFromCurve(
      FloatCurve=DataInterface() {propertyValues {Curve {keys [
          {arriveTangent=4.1785, interpMode=RCIM_Cubic, leaveTangent=4.1785, tangentMode=RCTM_User, time=0, value=0.625}
          {arriveTangent=0.9995, interpMode=RCIM_Cubic, leaveTangent=0.9995, tangentMode=RCTM_User, time=0.1, value=0.825}
          {arriveTangent=-0.0405, interpMode=RCIM_Cubic, leaveTangent=-0.0405, tangentMode=RCTM_User, time=1, value=1}
         ]}}}
      Scale Curve=2
     )))
   /SetVariables(Particles.FogSize:Vector2f=&Multiply_Vector2(A=$Particles.SpriteSize, B=&Vector2DFromFloat(Value=2))) {isSetParameters=true}
   /SetVariables(Particles.NewVector2D:Vector2f=&Vector2DFromFloat(Value=&Multiply_Float(
      A=&FloatFromCurve(FloatCurve=DataInterface() {propertyValues {Curve {keys [
           {arriveTangent=7.1136, interpMode=RCIM_Cubic, leaveTangent=7.1136, tangentMode=RCTM_User, time=0, value=0}
           {arriveTangent=1.4654, interpMode=RCIM_Cubic, leaveTangent=1.4654, time=0.18, value=1}
           <at 0.3814 0.5589>
           {arriveTangent=-0.0952, interpMode=RCIM_Cubic, leaveTangent=-0.0952, tangentMode=RCTM_User, time=0.7849, value=0.7663}
           <at 1 0.5 RCIM_Cubic>
          ]}}})
      B=$Particles.NewNiagaraFloat
     ))) {enabled=false, isSetParameters=true}
   &DynamicMaterialParameters(
    Index 0 Param 1=&RandomRangeFloat(Minimum=0, Maximum=1)
    Index 0 Param 2=&RandomRangeFloat(Minimum=0, Maximum=1)
    Index 0 Param 3=&FloatFromCurve(FloatCurve=DataInterface() {propertyValues {Curve {keys [<at 0 0>,<at 1 -0.2>]}}})
    Index 0 Param 4=0.25
    Module.Param0WriteEnabled=true
    Module.Param1WriteEnabled=true
    Module.Param2WriteEnabled=true
    Module.Param3WriteEnabled=true
   )
   &KillParticlesInVolume(Sphere Radius:NiagaraFloat=&Multiply_Float(
     A=$Engine.System.Age
     B=&FloatFromCurve(
      FloatCurve=DataInterface() {propertyValues {Curve {keys [
          {arriveTangent=83.6631, interpMode=RCIM_Cubic, leaveTangent=83.6631, tangentMode=RCTM_User, time=0, value=1}
          {arriveTangent=3.345, interpMode=RCIM_Cubic, leaveTangent=3.345, tangentMode=RCTM_User, time=1, value=25}
         ]}}}
      CurveIndex=$Engine.System.Age
      Scale Curve=5000
     )
    )) {enabled=false}
  ]
  rendererClasses [&NiagaraSpriteRendererProperties]
  renderers [
   {rendererClass=&NiagaraSpriteRendererProperties}
   {rendererClass=&NiagaraSpriteRendererProperties, rendererIndex=1}
  ]
 }
}
summary {
 systemName=NS_LuxurySupreme_FlashFadeOut
 emitters [
  {emitterName=FN_Mesh, rendererClasses [&NiagaraMeshRendererProperties, &NiagaraLightRendererProperties]}
  {emitterName=HT_LoopShockwave_CamF001, rendererClasses [&NiagaraMeshRendererProperties]}
  {emitterName=Pop_Sparks, simTarget=GPUComputeSim, rendererClasses [&NiagaraSpriteRendererProperties]}
  {emitterName=ConeLight001, rendererClasses [&NiagaraMeshRendererProperties]}
  {emitterName=HT_LoopShockwave_CamF, rendererClasses [&NiagaraMeshRendererProperties]}
 ]
 userVariables []
}
